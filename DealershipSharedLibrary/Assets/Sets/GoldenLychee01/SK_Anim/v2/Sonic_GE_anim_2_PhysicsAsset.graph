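# Physics asset for the Sonic_GE_anim_2 skeletal character in the GoldenLychee01 set, defining its collision bodies and joints.
# Single capsules cover the torso, limbs, fingers, head and six hair bones.
# Constraints chain them from the pelvis, with hair joined to the head and uniform swing and twist limit angles.
# The right side lacks clavicle and index, middle and ring tip bodies, so upperarm_r joins spine_03 directly.
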PhysicsAsset /DealershipSharedLibrary/Assets/Sets/GoldenLychee01/SK_Anim/v2/Sonic_GE_anim_2_PhysicsAsset
bodies {
 ball_l [{shapeType=Capsule, length=49.9007, radius=38.8019, center=<-3.0672 -9.9521 -38.5315>, rotation=<rot 0 -7.0945 -88.964>}]
 ball_r [{shapeType=Capsule, length=49.2086, radius=42.2168, center=<2.7813 -4.49 -35.3632>, rotation=<rot 0 -171.3307 -80.8055>}]
 calf_l [{shapeType=Capsule, length=55.1289, radius=51.2817, center=<-53.1587 -3.0932 0.4464>, rotation=<rot 41.4639 -90 -92.0136>}]
 calf_r [{shapeType=Capsule, length=62.4855, radius=50.8111, center=<49.874 -4.0393 -0.6536>, rotation=<rot 42.818 -90 -92.987>}]
 clavicle_l [{shapeType=Capsule, length=51.515, radius=26.7173, center=<6.6121 -30.0787 -5.3999>, rotation=<rot 0 132.7458 -82.1306>}]
 foot_l [{shapeType=Capsule, length=85.0464, radius=68.1112, center=<-37.0699 -5.1159 -29.0835>, rotation=<rot 26.5719 -90 5.5485>}]
 foot_r [{shapeType=Capsule, length=84.9216, radius=68.4602, center=<37.5567 -5.0884 -29.0053>, rotation=<rot 27.1434 -90 -5.3527>}]
 hair_a_1 [{shapeType=Capsule, length=148.4029, radius=105.1583, center=<28.2928 -33.4597 -33.2791>, rotation=<rot 0 -12.1526 -81.4064>}]
 hair_b_2 [{shapeType=Capsule, length=108.7877, radius=57.673, center=<-2.6026 13.1448 -12.9483>, rotation=<rot 0 -2.9358 -80.7128>}]
 hair_c_1 [{shapeType=Capsule, length=142.8689, radius=86.2274, center=<-18.4459 -7.454 16.8684>, rotation=<rot 0 175.6383 -69.8501>}]
 hair_d_1 [{shapeType=Capsule, length=144.78, radius=92.6607, center=<2.4206 -54.6827 -23.2873>, rotation=<rot 0 3.7314 -87.0078>}]
 hair_e_1 [{shapeType=Capsule, length=111.1317, radius=85.9925, center=<21.4623 -37.6795 6.3649>, rotation=<rot 0 179.4445 -71.9>}]
 hair_f_1 [{shapeType=Capsule, length=125.2786, radius=91.5568, center=<12.1206 -62.5001 0.8072>, rotation=<rot 0 -178.6159 -68.4027>}]
 hand_l [{shapeType=Capsule, length=70.6095, radius=49.9171, center=<42.3207 13.9522 -2.2031>, rotation=<rot 0 106.2348 -76.1653>}]
 hand_r [{shapeType=Capsule, length=66.2023, radius=51.2674, center=<-37.9533 -12.321 -1.9023>, rotation=<rot 0 106.9524 -88.8548>}]
 head [{shapeType=Capsule, length=129.499, radius=119.8022, center=<63.6979 12.4968 -7.7923>, rotation=<rot 0 89.041 -69.0334>}]
 index_01_l [{shapeType=Capsule, length=31.7011, radius=26.1325, center=<8.8706 -8.2213 4.6418>, rotation=<rot 0 -70.1998 -84.6473>}]
 index_01_r [{shapeType=Capsule, length=33.1658, radius=18.429, center=<-5.5636 4.5085 -2.3231>, rotation=<rot 0 -80.3593 -83.3504>}]
 index_02_l [{shapeType=Capsule, length=19.6714, radius=16.0142, center=<10.1359 1.6348 -2.1615>, rotation=<rot -0.311 -90 -93.2645>}]
 index_02_r [{shapeType=Capsule, length=27.1686, radius=15.6408, center=<-19.3627 4.9246 -0.0743>, rotation=<rot 0 -121.2678 -86.9134>}]
 index_03_l [{shapeType=Capsule, length=25.0436, radius=14.2858, center=<6.974 1.5982 -3.2982>, rotation=<rot 0 95.0662 -71.8218>}]
 lowerarm_l [{shapeType=Capsule, length=53.8751, radius=27.7929, center=<34.4828 4.1427 -2.3394>, rotation=<rot 0 84.4666 -85.9628>}]
 lowerarm_r [{shapeType=Capsule, length=51.5101, radius=27.7008, center=<-33.9409 -3.2248 3.0689>, rotation=<rot 0 86.2137 -85.5643>}]
 middle_01_l [{shapeType=Capsule, length=31.9158, radius=28.2797, center=<8.4256 -1.8384 -1.0218>, rotation=<rot 0 90.2311 -81.0413>}]
 middle_01_r [{shapeType=Capsule, length=31.5326, radius=25.9341, center=<-9.8497 3.7444 -0.2149>, rotation=<rot 0 91.7665 -86.0944>}]
 middle_02_l [{shapeType=Capsule, length=19.2316, radius=15.3437, center=<7.7085 -2.0042 -1.8123>, rotation=<rot -4.1482 -90 -95.127>}]
 middle_02_r [{shapeType=Capsule, length=26.5089, radius=16.4282, center=<-18.0029 7.5848 0.8423>, rotation=<rot 0 -129.1409 -88.9058>}]
 middle_03_l [{shapeType=Capsule, length=21.8122, radius=18.227, center=<11.7384 -2.4951 -1.6237>, rotation=<rot 0 90.4597 -61.6758>}]
 neck_01 [{shapeType=Capsule, length=46.3649, radius=39.3309, center=<11.736 11.4182 5.4536>, rotation=<rot 0 -32.4907 -82.7377>}]
 pelvis [{shapeType=Capsule, length=31.1637, radius=19.2691, center=<20.9831 -2.9617 8.3507>, rotation=<rot 0 -172.3753 -48.4401>}]
 pinky_01_l [{shapeType=Capsule, length=29.5021, radius=21.0484, center=<4.7951 2.0446 -5.7055>, rotation=<rot -18.7021 -90 43.307>}]
 pinky_01_r [{shapeType=Capsule, length=31.3294, radius=23.0754, center=<-3.4764 7.312 3.926>, rotation=<rot 0 93.1381 -65.8967>}]
 pinky_02_l [{shapeType=Capsule, length=21.6504, radius=15.6158, center=<10.7943 -6.256 1.4276>, rotation=<rot 14.5526 -90 -92.4802>}]
 pinky_02_r [{shapeType=Capsule, length=20.2195, radius=15.4009, center=<-10.1715 5.7719 -1.0737>, rotation=<rot 4.5762 -90 -87.1881>}]
 ring_01_l [{shapeType=Capsule, length=32.0998, radius=22.9592, center=<5.9595 -1.788 7.4798>, rotation=<rot 0 98.148 -77.9937>}]
 ring_01_r [{shapeType=Capsule, length=33.1084, radius=21.4973, center=<-8.9087 2.546 -6.2756>, rotation=<rot 0 96.2967 -81.8075>}]
 ring_02_l [{shapeType=Capsule, length=19.1624, radius=16.4378, center=<3.6936 -7.2543 -0.7975>, rotation=<rot 72.9806 83.5737 173.8529>}]
 ring_02_r [{shapeType=Capsule, length=27.5723, radius=15.1752, center=<-16.9721 2.8671 -0.6163>, rotation=<rot 0 -116.5328 -88.2455>}]
 ring_03_l [{shapeType=Capsule, length=23.1796, radius=13.7906, center=<3.9256 8.3544 0.2792>, rotation=<rot 0 -105.2325 -87.488>}]
 spine_01 [{shapeType=Capsule, length=53.6136, radius=35.9656, center=<6.6878 -38.1261 3.3485>, rotation=<rot 0 -7.64 -84.3053>}]
 spine_02 [{shapeType=Capsule, length=49.2027, radius=43.8956, center=<8.8077 -16.9702 -0.0407>, rotation=<rot 0 -145.9793 -87.6797>}]
 spine_03 [{shapeType=Capsule, length=83.8783, radius=54.8639, center=<-4.3115 -36.1318 2.1773>, rotation=<rot 0 156.6955 -84.0213>}]
 thigh_l [{shapeType=Capsule, length=62.0473, radius=20.6995, center=<-56.9296 3.5375 -1.5572>, rotation=<rot 0 -84.3164 -84.3002>}]
 thigh_r [{shapeType=Capsule, length=60.2205, radius=19.042, center=<50.4622 3.4867 0.2087>, rotation=<rot 0 86.9568 -88.0354>}]
 thumb_01_l [{shapeType=Capsule, length=34.1449, radius=26.6718, center=<5.2366 -4.9558 5.738>, rotation=<rot 0 67.4916 -73.0117>}]
 thumb_01_r [{shapeType=Capsule, length=36.5875, radius=26.8982, center=<-9.4567 4.8148 -6.6193>, rotation=<rot 0 69.5985 -83.5199>}]
 thumb_02_l [{shapeType=Capsule, length=26.2293, radius=17.8676, center=<14.834 -3.5109 1.6747>, rotation=<rot 0 -89.6579 -71.7863>}]
 thumb_02_r [{shapeType=Capsule, length=23.4166, radius=19.3023, center=<-10.8936 0.7169 -2.4967>, rotation=<rot -9.7104 -90 -44.5395>}]
 thumb_03_l [{shapeType=Capsule, length=18.0052, radius=15.4846, center=<9.8491 2.4147 -1.7848>, rotation=<rot 0 85.7415 -77.6127>}]
 thumb_03_r [{shapeType=Capsule, length=20.7882, radius=14.6039, center=<-8.3216 -0.67 1.4434>, rotation=<rot 0 -79.8209 -85.601>}]
 upperarm_l [{shapeType=Capsule, length=36.8596, radius=10.3603, center=<30.2096 1.2902 0.5498>, rotation=<rot 0 93.2411 -83.6364>}]
 upperarm_r [{shapeType=Capsule, length=38.1578, radius=10.2795, center=<-30.3474 0.0009 -1.7813>, rotation=<rot 0 89.3114 -84.514>}]
}
constraints [
 {bone1Name=spine_01, bone2Name=pelvis, swing1LimitDegrees=45, swing2LimitDegrees=45, twistLimitDegrees=45}
 {bone1Name=spine_02, bone2Name=spine_01, swing1LimitDegrees=45, swing2LimitDegrees=45, twistLimitDegrees=45}
 {bone1Name=spine_03, bone2Name=spine_02, swing1LimitDegrees=45, swing2LimitDegrees=45, twistLimitDegrees=45}
 {bone1Name=clavicle_l, bone2Name=spine_03, swing1LimitDegrees=45, swing2LimitDegrees=45, twistLimitDegrees=45}
 {bone1Name=upperarm_l, bone2Name=clavicle_l, swing1LimitDegrees=45, swing2LimitDegrees=45, twistLimitDegrees=45}
 {bone1Name=lowerarm_l, bone2Name=upperarm_l, swing1LimitDegrees=45, swing2LimitDegrees=45, twistLimitDegrees=45}
 {bone1Name=hand_l, bone2Name=lowerarm_l, swing1LimitDegrees=45, swing2LimitDegrees=45, twistLimitDegrees=45}
 {bone1Name=index_01_l, bone2Name=hand_l, swing1LimitDegrees=45, swing2LimitDegrees=45, twistLimitDegrees=45}
 {bone1Name=index_02_l, bone2Name=index_01_l, swing1LimitDegrees=45, swing2LimitDegrees=45, twistLimitDegrees=45}
 {bone1Name=index_03_l, bone2Name=index_02_l, swing1LimitDegrees=45, swing2LimitDegrees=45, twistLimitDegrees=45}
 {bone1Name=middle_01_l, bone2Name=hand_l, swing1LimitDegrees=45, swing2LimitDegrees=45, twistLimitDegrees=45}
 {bone1Name=middle_02_l, bone2Name=middle_01_l, swing1LimitDegrees=45, swing2LimitDegrees=45, twistLimitDegrees=45}
 {bone1Name=middle_03_l, bone2Name=middle_02_l, swing1LimitDegrees=45, swing2LimitDegrees=45, twistLimitDegrees=45}
 {bone1Name=pinky_01_l, bone2Name=hand_l, swing1LimitDegrees=45, swing2LimitDegrees=45, twistLimitDegrees=45}
 {bone1Name=pinky_02_l, bone2Name=pinky_01_l, swing1LimitDegrees=45, swing2LimitDegrees=45, twistLimitDegrees=45}
 {bone1Name=ring_01_l, bone2Name=hand_l, swing1LimitDegrees=45, swing2LimitDegrees=45, twistLimitDegrees=45}
 {bone1Name=ring_02_l, bone2Name=ring_01_l, swing1LimitDegrees=45, swing2LimitDegrees=45, twistLimitDegrees=45}
 {bone1Name=ring_03_l, bone2Name=ring_02_l, swing1LimitDegrees=45, swing2LimitDegrees=45, twistLimitDegrees=45}
 {bone1Name=thumb_01_l, bone2Name=hand_l, swing1LimitDegrees=45, swing2LimitDegrees=45, twistLimitDegrees=45}
 {bone1Name=thumb_02_l, bone2Name=thumb_01_l, swing1LimitDegrees=45, swing2LimitDegrees=45, twistLimitDegrees=45}
 {bone1Name=thumb_03_l, bone2Name=thumb_02_l, swing1LimitDegrees=45, swing2LimitDegrees=45, twistLimitDegrees=45}
 {bone1Name=upperarm_r, bone2Name=spine_03, swing1LimitDegrees=45, swing2LimitDegrees=45, twistLimitDegrees=45}
 {bone1Name=lowerarm_r, bone2Name=upperarm_r, swing1LimitDegrees=45, swing2LimitDegrees=45, twistLimitDegrees=45}
 {bone1Name=hand_r, bone2Name=lowerarm_r, swing1LimitDegrees=45, swing2LimitDegrees=45, twistLimitDegrees=45}
 {bone1Name=index_01_r, bone2Name=hand_r, swing1LimitDegrees=45, swing2LimitDegrees=45, twistLimitDegrees=45}
 {bone1Name=index_02_r, bone2Name=index_01_r, swing1LimitDegrees=45, swing2LimitDegrees=45, twistLimitDegrees=45}
 {bone1Name=middle_01_r, bone2Name=hand_r, swing1LimitDegrees=45, swing2LimitDegrees=45, twistLimitDegrees=45}
 {bone1Name=middle_02_r, bone2Name=middle_01_r, swing1LimitDegrees=45, swing2LimitDegrees=45, twistLimitDegrees=45}
 {bone1Name=pinky_01_r, bone2Name=hand_r, swing1LimitDegrees=45, swing2LimitDegrees=45, twistLimitDegrees=45}
 {bone1Name=pinky_02_r, bone2Name=pinky_01_r, swing1LimitDegrees=45, swing2LimitDegrees=45, twistLimitDegrees=45}
 {bone1Name=ring_01_r, bone2Name=hand_r, swing1LimitDegrees=45, swing2LimitDegrees=45, twistLimitDegrees=45}
 {bone1Name=ring_02_r, bone2Name=ring_01_r, swing1LimitDegrees=45, swing2LimitDegrees=45, twistLimitDegrees=45}
 {bone1Name=thumb_01_r, bone2Name=hand_r, swing1LimitDegrees=45, swing2LimitDegrees=45, twistLimitDegrees=45}
 {bone1Name=thumb_02_r, bone2Name=thumb_01_r, swing1LimitDegrees=45, swing2LimitDegrees=45, twistLimitDegrees=45}
 {bone1Name=thumb_03_r, bone2Name=thumb_02_r, swing1LimitDegrees=45, swing2LimitDegrees=45, twistLimitDegrees=45}
 {bone1Name=neck_01, bone2Name=spine_03, swing1LimitDegrees=45, swing2LimitDegrees=45, twistLimitDegrees=45}
 {bone1Name=head, bone2Name=neck_01, swing1LimitDegrees=45, swing2LimitDegrees=45, twistLimitDegrees=45}
 {bone1Name=hair_a_1, bone2Name=head, swing1LimitDegrees=45, swing2LimitDegrees=45, twistLimitDegrees=45}
 {bone1Name=hair_b_2, bone2Name=head, swing1LimitDegrees=45, swing2LimitDegrees=45, twistLimitDegrees=45}
 {bone1Name=hair_c_1, bone2Name=head, swing1LimitDegrees=45, swing2LimitDegrees=45, twistLimitDegrees=45}
 {bone1Name=hair_d_1, bone2Name=head, swing1LimitDegrees=45, swing2LimitDegrees=45, twistLimitDegrees=45}
 {bone1Name=hair_e_1, bone2Name=head, swing1LimitDegrees=45, swing2LimitDegrees=45, twistLimitDegrees=45}
 {bone1Name=hair_f_1, bone2Name=head, swing1LimitDegrees=45, swing2LimitDegrees=45, twistLimitDegrees=45}
 {bone1Name=thigh_l, bone2Name=pelvis, swing1LimitDegrees=45, swing2LimitDegrees=45, twistLimitDegrees=45}
 {bone1Name=calf_l, bone2Name=thigh_l, swing1LimitDegrees=45, swing2LimitDegrees=45, twistLimitDegrees=45}
 {bone1Name=foot_l, bone2Name=calf_l, swing1LimitDegrees=45, swing2LimitDegrees=45, twistLimitDegrees=45}
 {bone1Name=ball_l, bone2Name=foot_l, swing1LimitDegrees=45, swing2LimitDegrees=45, twistLimitDegrees=45}
 {bone1Name=thigh_r, bone2Name=pelvis, swing1LimitDegrees=45, swing2LimitDegrees=45, twistLimitDegrees=45}
 {bone1Name=calf_r, bone2Name=thigh_r, swing1LimitDegrees=45, swing2LimitDegrees=45, twistLimitDegrees=45}
 {bone1Name=foot_r, bone2Name=calf_r, swing1LimitDegrees=45, swing2LimitDegrees=45, twistLimitDegrees=45}
 {bone1Name=ball_r, bone2Name=foot_r, swing1LimitDegrees=45, swing2LimitDegrees=45, twistLimitDegrees=45}
]
